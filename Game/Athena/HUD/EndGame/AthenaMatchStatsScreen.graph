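# Athena end-game match stats screen widget, based on FortMatchStatsScreen, hosting the AthenaMatchStatsPartial in a bottom-right safe zone.
# On activation or stats update, it enables Back input and passes the owning player controller's match report to the partial.
# Construct registers a Back action handler.
# A border touch zone, collapsed by default, handles primary press or touch through HandleBack, which pops the content widget.
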WidgetBlueprint /Game/Athena/HUD/EndGame/AthenaMatchStatsScreen
graphs [
 EventGraph {nodes [
   n1 = Event EventUpdateStats() {is=title K2Node_Event}
   InputRefresh(execute=[@n1.then, @n4.then])
   > n2 = Cast To Athena_PlayerController!(execute, Object=GetOwningPlayer()) -> (then, CastFailed, AsAthena Player Controller:Athena Player Controller Object Reference) {is=title K2Node_DynamicCast}
   > n3 = GetMatchReport!(execute, self:Fort Player Controller Athena Object Reference=@n2) -> (then, ReturnValue:Athena Player Match Report Object Reference)
   > SetMatchReport!(
    execute
    self:Athena Match Stats Partial Object Reference=Get AthenaMatchStatsPartial() -> (AthenaMatchStatsPartial:Athena Match Stats Partial Object Reference)
    Report:Athena Player Match Report Object Reference=@n3
   ) -> (then)
   n4 = Event On Activated()
   Event Construct()
   > InputSetup()
  ]}
 HandleBack {nodes [
   HandleBack() {is=title K2Node_FunctionEntry}
   > PopContentWidget!(execute) -> (then)
   > Return Node!(execute, PassThrough:bool=false)
  ]}
 InputRefresh {nodes [
   InputRefresh() {is=title K2Node_FunctionEntry}
   > SetActionHandlerStateFromHandle!(
    execute
    self:Common Activatable Panel Legacy Object Reference
    InputActionRow:Data Table Row Handle Structure=Make Data Table Row Handle!(DataTable:Data Table Object Reference=&InputActionDataTable, RowName:Name=Back) -> (DataTableRowHandle:Data Table Row Handle Structure) {is=title K2Node_MakeStruct}
    State:EInputActionState Enum=Enabled
   ) -> (then)
  ]}
 InputSetup {nodes [
   InputSetup() {is=title K2Node_FunctionEntry}
   > RegisterBack()
  ]}
 On_BorderTouchZone_MouseButtonDown {nodes [
   n1 = On_BorderTouchZone_MouseButtonDown!() -> (then, MyGeometry:Geometry Structure, MouseEvent:"Pointer Event Structure (by ref)") {is=title K2Node_FunctionEntry}
   > n2 = Handle Primary Pointer Button or Touch!(Execute:Exec, Mouse Event:"Pointer Event Structure (by ref)"=@n1.MouseEvent) -> (If Primary:Exec, Continue:Exec, Reply:Event Reply Structure) {is=title K2Node_MacroInstance}
   Return Node!(execute=@n2.Continue, ReturnValue:Event Reply Structure=@n2)
   > HandleBack!(execute, self) -> (then, PassThrough:bool) {is=title K2Node_CallFunction}
  ]}
 RegisterBack {nodes [
   RegisterBack() {is=title K2Node_FunctionEntry}
   > SetInputActionHandler!(
    execute
    self:Common Activatable Panel Legacy Object Reference
    InputActionRow:Data Table Row Handle Structure=Make Data Table Row Handle(DataTable=&InputActionDataTable, RowName=Back)
    CommitedEvent:Delegate=Create Event()
   ) -> (then)
  ]}
]
variables []
widgets {
 info {inheritedWidgetCount=0, parentClass=/Script/FortniteUI.FortMatchStatsScreen, rootWidgetClass=&Overlay, widgetCount=5}
 widgets [
  Overlay_0 {widgetClassPath=&Overlay, widget=Overlay_0}
  HorizontalBox_0 {
   widgetClassPath=&HorizontalBox
   parent=Overlay_0
   widget=HorizontalBox_0
   slot {class=OverlaySlot, properties {horizontalAlignment=HAlign_Fill, verticalAlignment=VAlign_Fill}}
  }
  BorderTouchZone {
   widgetClassPath=&CommonBorder
   bIsVariable=true
   parent=HorizontalBox_0
   widget=BorderTouchZone
   properties {onMouseButtonDownEvent="(null).On_BorderTouchZone_MouseButtonDown_0", visibility=Collapsed, padding=<ltrb 0 0 0 0>}
   slot {class=HorizontalBoxSlot, properties {size {sizeRule=Fill, value=1}}}
  }
  SafeZone_0 {
   widgetClassPath=&SafeZone
   parent=HorizontalBox_0
   widget=SafeZone_0
   properties {padBottom=false, padLeft=false}
   slot {class=HorizontalBoxSlot, properties {horizontalAlignment=HAlign_Right, verticalAlignment=VAlign_Bottom}}
  }
  AthenaMatchStatsPartial {
   widgetClassPath=&/AthenaMatchStatsPartial
   bIsVariable=true
   parent=SafeZone_0
   widget=AthenaMatchStatsPartial
   slot {class=SafeZoneSlot}
  }
 ]
}
