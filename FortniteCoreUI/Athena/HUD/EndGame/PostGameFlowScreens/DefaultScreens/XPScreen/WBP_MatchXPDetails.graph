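# Post-game XP screen widget, built on FortMatchXPDetails, that shows match stats beside an Accolades panel.
# The layout places the inherited match stats section left of a titled accolades column inside a top-only safe zone.
# Border_Accolades holds a WidgetSlottingHostOverlay tagged for the dynamic AccoladeList slot.
# The graph's only logic adds any created extension widget to that border; CachedAccoladeList is declared but unused here.
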WidgetBlueprint /FortniteCoreUI/Athena/HUD/EndGame/PostGameFlowScreens/DefaultScreens/XPScreen/WBP_MatchXPDetails
graphs [EventGraph {nodes [
   n1 = Event Handle Extension Widget Created!() -> (OutputDelegate, then, Widget:User Widget Object Reference, SlotType:EUIExtensionSlot Enum=Invalid) {is=title K2Node_Event}
   > AddChild(self=Get Border_Accolades() -> (Border_Accolades:Common Border Object Reference), Content=@n1.Widget)
  ]}]
variables [CachedAccoladeList {type=UserWidget, default=null}]
widgets {
 info {inheritedWidgetCount=1, parentClass=/Script/FortniteUI.FortMatchXPDetails, rootWidgetClass=&Overlay, widgetCount=13}
 widgets [
  Overlay_0 {widgetClassPath=&Overlay, widget=Overlay_0}
  SizeBox_0 {
   widgetClassPath=&SizeBox
   parent=Overlay_0
   widget=SizeBox_0
   properties {bOverride_HeightOverride=true, heightOverride=715.6155}
   slot {class=OverlaySlot, properties {horizontalAlignment=HAlign_Center, verticalAlignment=VAlign_Center, padding=<ltrb 0 100 0 0>}}
  }
  HorizontalBox_0 {widgetClassPath=&HorizontalBox, parent=SizeBox_0, widget=HorizontalBox_0, slot {class=SizeBoxSlot}}
  Section_MatchStats {
   widgetClassPath=/Game/Athena/HUD/EndGame/AthenaMatchStatsScreen
   bInherited=true
   bIsVariable=true
   parent=HorizontalBox_0
   widget=Section_MatchStats
   slot {class=HorizontalBoxSlot, properties {horizontalAlignment=HAlign_Center, verticalAlignment=VAlign_Top}}
  }
  Spacer_33 {widgetClassPath=&Spacer, parent=HorizontalBox_0, widget=Spacer_33, properties {size=<62.4919 1>}, slot {class=HorizontalBoxSlot}}
  SafeZone_0 {
   widgetClassPath=&SafeZone
   bIsVariable=true
   parent=HorizontalBox_0
   widget=SafeZone_0
   properties {padBottom=false, padLeft=false, padRight=false}
   slot {class=HorizontalBoxSlot}
  }
  SizeBox_1 {
   widgetClassPath=&SizeBox
   parent=SafeZone_0
   widget=SizeBox_1
   properties {bOverride_WidthOverride=true, widthOverride=505.5864}
   slot {class=SafeZoneSlot}
  }
  VerticalBox_0 {widgetClassPath=&VerticalBox, parent=SizeBox_1, widget=VerticalBox_0, slot {class=SizeBoxSlot}}
  CommonTextBlock_0 {
   widgetClassPath=&CommonTextBlock
   parent=VerticalBox_0
   widget=CommonTextBlock_0
   properties {
    style=/Game/Athena/HUD/TextStyles/TextStyle-HUD-MapScreen_Title.TextStyle-HUD-MapScreen_Title_C
    text=Accolades
    textTransformPolicy=ToUpper
    font {
     bForceMonospaced=false
     bMaterialIsStencil=false
     fontMaterial=null
     fontObject=&Burbank
     letterSpacing=0
     monospacedWidth=1
     size=18
     skewAmount=0
     typefaceFontName=BurbankRegular-Black
     outlineSettings {
      bApplyOutlineToDropShadows=true
      bMiteredCorners=false
      bSeparateFillAlpha=true
      outlineMaterial=&M-FontShadow
      outlineSize=3
      outlineColor=<rgba 0 0 0 0.5>
     }
    }
    renderTransform {angle=0, scale=<1 1>, shear=<-8 0>, translation=<0 0>}
    shadowOffset=<0 0>
   }
   slot {class=VerticalBoxSlot, properties {horizontalAlignment=HAlign_Left}}
  }
  Overlay_1 {
   widgetClassPath=&Overlay
   parent=VerticalBox_0
   widget=Overlay_1
   slot {class=VerticalBoxSlot, properties {size {sizeRule=Fill, value=1}}}
  }
  Image_FlatBG {
   widgetClassPath=&Image
   bIsVariable=true
   parent=Overlay_1
   widget=Image_FlatBG
   properties {renderOpacity=0, colorAndOpacity=<rgba 0.0052 0.0052 0.0056 0.5>}
   slot {class=OverlaySlot, properties {horizontalAlignment=HAlign_Fill, verticalAlignment=VAlign_Fill}}
  }
  Border_Accolades {
   widgetClassPath=&CommonBorder
   bIsVariable=true
   parent=Overlay_1
   widget=Border_Accolades
   properties {padding=<ltrb 0 0 0 0>}
   slot {class=OverlaySlot, properties {horizontalAlignment=HAlign_Fill, verticalAlignment=VAlign_Fill, padding=<ltrb 10 10 10 20>}}
  }
  WidgetSlottingHostOverlay_Accolades {
   widgetClassPath=/Script/WidgetSlotting.WidgetSlottingHostOverlay
   parent=Border_Accolades
   widget=WidgetSlottingHostOverlay_Accolades
   properties {
    slottedWidgetDefaultHorizontalAlignment=HAlign_Fill
    slottedWidgetDefaultVerticalAlignment=VAlign_Fill
    hostTag {tagName=DynamicUI.HUD.Fort.Athena.PostMatch.AccoladeList.Slot}
   }
   slot {class=BorderSlot, properties {padding=<ltrb 0 0 0 0>}}
  }
 ]
}
