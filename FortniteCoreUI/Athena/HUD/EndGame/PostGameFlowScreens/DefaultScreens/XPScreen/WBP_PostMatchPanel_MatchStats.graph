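# Post-match panel widget that shows the player's match XP details in the end-game flow.
# On activation and deactivation it toggles HUD elements tagged HideOnMatchStats and broadcasts an activation-changed event through the global event router.
# The first activation after construction populates the MatchXPDetails child widget, with a flag preventing repeat updates.
# Map-screen override functions all return false.
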WidgetBlueprint /FortniteCoreUI/Athena/HUD/EndGame/PostGameFlowScreens/DefaultScreens/XPScreen/WBP_PostMatchPanel_MatchStats
graphs [
 EventGraph {nodes [
   Event On Deactivated()
   > UpdateHideOnMatchStatsVisibility!(execute, self, bHideElements:bool=false) -> (then)
   > Broadcast Event!(
    execute
    self:Gameplay Event Router Component Object Reference=GetGlobalEventRouter()
    EventData:Fort Event Fortnite Core UIMatch Stats Panel Activation Changed Structure=Make Fort Event Fortnite Core UIMatch Stats Panel Activation Changed!(Activated:bool=False) -> (FortEvent_FortniteCoreUIMatchStatsPanelActivationChanged:Fort Event Fortnite Core UIMatch Stats Panel Activation Changed Structure) {is=title K2Node_MakeStruct}
    EventContext:Object Reference=GetOwningPlayer()
   ) -> (then, ReturnValue:bool)
   Event OnActivatePanel() {is=title K2Node_Event}
   > UpdateHideOnMatchStatsVisibility(bHideElements=true)
   > Broadcast Event(
    self=GetGlobalEventRouter()
    EventData=Make Fort Event Fortnite Core UIMatch Stats Panel Activation Changed(Activated=true)
    EventContext=GetOwningPlayer()
   )
   > n1 = Branch(Condition=Get IsStatsPopulated() -> (IsStatsPopulated:bool))
   Set IsStatsPopulated(execute=@n1.else, IsStatsPopulated:bool=true) -> (Output_Get:bool)
   > UpdateMatchStats(self:Fort Match XPDetails Object Reference=Get MatchXPDetails() -> (MatchXPDetails:WBP Match XPDetails Object Reference))
   Event Construct()
   > Parent: Construct()
   > Set IsStatsPopulated(IsStatsPopulated=false)
  ]}
 OnBlockMapActionByKey {nodes [
   OnBlockMapActionByKey!() -> (then, InKey:"Key Structure (by ref)") {is=title K2Node_FunctionEntry}
   > Return Node!(execute, bOutMapActionBlocked:bool=false, ReturnValue:bool=false)
  ]}
 OnHandleMapActionKey {nodes [
   OnHandleMapActionKey!() -> (then, InKey:"Key Structure (by ref)", bKeyDown:bool) {is=title K2Node_FunctionEntry}
   > Return Node!(execute, bOutMapActionHandled:bool=false, ReturnValue=false)
  ]}
 OnHandleMapBackAction {nodes [
   OnHandleMapBackAction() {is=title K2Node_FunctionEntry}
   > Return Node!(execute, bOutBackActionHandled:bool=false, ReturnValue=false)
  ]}
 OnShouldCreateMapScreenTab {nodes [
   OnShouldCreateMapScreenTab!() -> (then, WorldContextObject:Object Reference) {is=title K2Node_FunctionEntry}
   > Return Node!(execute, bOutShouldCreateMapScreenTab:bool=false, ReturnValue=false)
  ]}
 OnShouldDisplayBackActionInActionBar {nodes [
   OnShouldDisplayBackActionInActionBar() {is=title K2Node_FunctionEntry}
   > Return Node!(execute, bOutShouldDisplayBackActionInActionBar:bool=false, ReturnValue=false)
  ]}
 OnShouldShowTabOnMapScreenForMinigame {nodes [
   OnShouldShowTabOnMapScreenForMinigame!() -> (then, Minigame:Fort Minigame Object Reference) {is=title K2Node_FunctionEntry}
   > Return Node!(execute, bOutShouldShowTabOnMapScreenForMinigame:bool=false, ReturnValue=false)
  ]}
 UpdateHideOnMatchStatsVisibility {nodes [
   n1 = UpdateHideOnMatchStatsVisibility!() -> (then, bHideElements:bool=~) {is=title K2Node_FunctionEntry}
   > n2 = Cast To FortPlayerControllerAthena(Object=GetOwningPlayer())
   > SetHUDElementVisibilityWithContext!(
    execute
    self:Fort Player Controller Object Reference=@n2
    HUDElementTags:"Gameplay Tag Container Structure (by ref)"=MakeGameplayTagContainerFromTag!(SingleTag:Gameplay Tag Structure=(TagName="HUD.Athena.HideOnMatchStats")) -> (ReturnValue:Gameplay Tag Container Structure)
    bHideElements:bool=@n1
    ReasonName:Name=HideOnMatchStats tag is setting visibility
    LogContext:String
   ) -> (then)
  ]}
]
variables [IsStatsPopulated {type=boolean}]
widgets {
 info {inheritedWidgetCount=0, parentClass=/Script/FortniteUI.AthenaPostMatchPanel, rootWidgetClass=&Overlay, widgetCount=2}
 widgets [
  Overlay_0 {widgetClassPath=&Overlay, widget=Overlay_0}
  MatchXPDetails {
   widgetClassPath=&/WBP_MatchXPDetails
   bIsVariable=true
   parent=Overlay_0
   widget=MatchXPDetails
   properties {padding=<ltrb 120 0 0 0>}
   slot {class=OverlaySlot, properties {verticalAlignment=VAlign_Center}}
  }
 ]
}
