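# Sound cue for the survivor tick on the ranked post-game UI screen, judging by its asset and wave names.
# A single wave player plays one tick sample directly as the root, with reduced volume under the SFX sound class.
# Its own concurrency override caps overlapping instances at two and stops the oldest, and virtualization is disabled.
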SoundCue /FortUIAudioContent/SoundCues/Fort_UI_PostGame_Ranked_Tick_Survivor_Cue
root=WavePlayer_0
nodes [{id=WavePlayer_0, properties {soundWaveAssetPtr=/FortUIAudioContent/Waves/UI_Ranked_PostMatch_Tick_Survivor_01}}]
properties {
 bOverrideConcurrency=true
 maxDistance=2097152
 soundClassObject=&SFX
 virtualizationMode=Disabled
 volumeMultiplier=0.6
 concurrencyOverrides {maxCount=2, resolutionRule=StopOldest}
}
